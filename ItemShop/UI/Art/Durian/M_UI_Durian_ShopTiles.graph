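# UI material for Durian item shop tiles, drawing an animated background that alternates between an orange-red and a pink-red palette.
# A time-driven sine, scaled by the "tenminutes" parameter, is smoothstepped to blend the palettes, each built from layered exponential radial gradients over a dark base.
# Named reroutes supply screen-fitted UVs, pixelized gradient UVs, a linear gradient and an LED dot layer whose circle grid is modulated by a mirrored, panning ping-pong pattern.
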Material /ItemShop/UI/Art/Durian/M_UI_Durian_ShopTiles
graph {
 properties {bUsedWithStaticMesh=false, materialDomain=MD_UI, pixelDepthOffsetMode=PDOM_Legacy, usedShadingModels=Default Lit}
 nodes [
  n1 = Time()
  n2 = ScalarParameter() {defaultValue=0.2, parameterName=tenminutes}
  n3 = Multiply(A=Sine(None=Multiply(A=@n1, B=@n2)) {period=10}) {constB=20}
  Ceil(None=@n3)
  n4 = Constant3Vector() {constant=<rgba 0.0208 0.003 0.0054 1>}
  n5 = NamedRerouteUsage()
  n6 = Constant() {r=0.45}
  n7 = Constant() {r=6}
  n8 = LinearInterpolate(
   A=LinearInterpolate(
    A=Constant3Vector() {constant=<rgba 0.4948 0 0.0154 1>}
    B=Constant3Vector() {constant=<rgba 1 0.3534 0.0076 1>}
    Alpha=MaterialFunctionCall(UVs=@n5, Radius=@n6, Density=@n7) {materialFunction=&RadialGradientExponential}.RadialGradientExponential
   )
   B=Constant3Vector() {constant=<rgba 0.724 0.2213 0.0264 1>}
   Alpha=MaterialFunctionCall(UVs=@n5, Radius=Constant() {r=0.43}, Density=Constant() {r=18}) {materialFunction=&RadialGradientExponential}.RadialGradientExponential
  )
  n9 = NamedRerouteUsage()
  n10 = Constant3Vector() {constant=<rgba 0.2188 0.0023 0.0023 0>}
  n11 = NamedRerouteUsage()
  n12 = NamedRerouteUsage()
  n13 = Constant() {r=0.5}
  n14 = Constant() {r=6}
  n15 = LinearInterpolate(
   A=LinearInterpolate(
    A=Constant3Vector() {constant=<rgba 0.4271 0 0.0507 1>}
    B=Constant3Vector() {constant=<rgba 1 0.099 0.1591 1>}
    Alpha=MaterialFunctionCall(UVs=@n12, Radius=@n13, Density=@n14) {materialFunction=&RadialGradientExponential}.RadialGradientExponential
   )
   B=Constant3Vector() {constant=<rgba 0.5521 0.1679 0.0401 1>}
   Alpha=MaterialFunctionCall(UVs=NamedRerouteUsage(), Radius=Constant() {r=0.37}, Density=Constant() {r=18}) {materialFunction=&RadialGradientExponential}.RadialGradientExponential
  )
  n16 = NamedRerouteUsage()
  n17 = Constant3Vector() {constant=<rgba 0.1771 0.0091 0.0144 1>}
  n18 = NamedRerouteUsage()
  LinearInterpolate(
   A=LinearInterpolate(A=LinearInterpolate(A=@n4, B=@n8, Alpha=@n9), B=Multiply(A=@n10, B=@n11), Alpha=@n11)
   B=LinearInterpolate(A=LinearInterpolate(A=@n4, B=@n15, Alpha=@n16), B=Multiply(A=@n17, B=@n18), Alpha=@n18)
   Alpha=Saturate(None=SmoothStep(Value=@n3) {constMax=8, constMin=0.2})
  )
  n19 = NamedRerouteUsage()
  n20 = Constant() {r=0.75}
  n21 = ComponentMask(None=OneMinus(None=MaterialFunctionCall(UVs=@n19, "Rotation Angle (0-1)"=@n20) {materialFunction=&CustomRotator}.Rotated Values)) {r=true}
  LinearInterpolate(A=Constant() {r=1}, B=Constant() {r=1}, Alpha=@n21)
  n22 = NamedRerouteUsage()
  n23 = Constant()
  n24 = Constant() {r=0.5}
  n25 = Constant() {r=-0.25}
  NamedRerouteDeclaration(None=SmoothStep(Value=MaterialFunctionCall(UVs=@n22, GradientStart=@n23, GradientEnd=@n24, RotationAngle=@n25) {materialFunction=&MF_UI_LinearGradient}.Result) {constMax=1.5}) {name=LinearGradient, nodeColor=<rgba 0.5 1 0 1>}
  n26 = AppendVector(A=Constant(), B=Constant() {r=-0.36})
  NamedRerouteDeclaration(None=MaterialFunctionCall!(UV=Add(A=@n22, B=@n26), Resolution=Constant() {r=200}) {materialFunction=/Game/Weapons/FORT_Melee/Pickaxe_Veiled/FX/MF_UVPixelizer}.Result) {name=RadialGradientUVs, nodeColor=<rgba 0 1 0.4141 1>}
  n27 = MaterialFunctionCall() {materialFunction=&MF_UI_UVScaling}
  NamedRerouteDeclaration(None=@n27.ScaleY) {name=UVFit, nodeColor=<rgba 1 0 0.6094 1>}
  n28 = MaterialFunctionCall!(UV=NamedRerouteUsage(), Offset=Constant2Vector(), Rotation, Scale=Constant() {r=0.75}, MirrorY=StaticBool() {value=true}) {materialFunction=/FortUILibrary/MaterialFunctions/Basics/MF_UI_Symmetry_2x1_DEPRECATED}
  n29 = Time() {period=10}
  n30 = MaterialFunctionCall!(
   Value=MaterialFunctionCall(UVs=Panner(Coordinate=@n28.Result, Time=@n29) {speedX=0.03}, "Rotation Angle (0-1)"=Constant() {r=0.35}) {materialFunction=&CustomRotator}.Rotated Values
   Scale=Constant() {r=8}
   Height
  ) {materialFunction=/FortUILibrary/MaterialFunctions/Basics/MF_UI_PingPong}
  n31 = OneMinus(None=ComponentMask(None=Clamp(None=@n30.Result) {minDefault=0.35}) {r=true})
  n32 = SmoothStep(Value=@n31) {constMax=0.5}
  n33 = Frac(None=Multiply(A=NamedRerouteUsage(), B=Constant() {r=110}))
  n34 = MaterialFunctionCall!(
   x=SmoothStep(
    Min=@n31
    Max=Add(A=@n31) {constB=0.001}
    Value=MaterialFunctionCall!(In Texcoords=@n33) {materialFunction=/Game/Effects/Fort_Effects/MaterialFunctions/MF_Circle_SDF}.Result
   )
   size=@n21
  ) {materialFunction=/Game/Effects/Fort_Effects/MaterialFunctions/MF_Fill}
  NamedRerouteDeclaration(None=Multiply(A=Min(A=@n32, B=@n34.Result)) {constB=0.15}) {name=LED Dots, nodeColor=<rgba 0.2188 1 0 1>}
  n35 = MaterialFunctionCall!(ScreenUV, TextureSize, ScalePivotPoint) {materialFunction=/FortUILibrary/MaterialFunctions/Basics/MF_UI_ScreenSpaceUVScaling}
  NamedRerouteDeclaration(None=@n35.Scaled to Fill) {name=UVFIll, nodeColor=<rgba 0 0.625 1 1>}
  NamedRerouteDeclaration(None=@n35.ScaleY) {name=Scale Y, nodeColor=<rgba 0 1 0.9062 1>}
  NamedRerouteDeclaration(None=@n27.Scaled To Fill) {name=ScalingUVFill, nodeColor=<rgba 0 1 0.8594 1>}
  NamedRerouteUsage()
 ]
}
